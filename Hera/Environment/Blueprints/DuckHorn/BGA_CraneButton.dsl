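; Interactable crane button offering a "Vote to Add Rocks" prompt, usable once per pawn.
; Interacting adds the pawn to a used-horn list, which blocks repeat use, then calls PlayBeep, which plays audio only where the actor lacks authority.
; On authority, BeginPlay binds to the attachment prop's OnDied so the button destroys itself when that prop dies.
; No vote tally or rock-adding logic appears in this Blueprint.
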(event EventBeginPlay
  (bind _attachmentprop (Variables|Default|GetAttachmentProp))
  (Networking|SwitchHasAuthority
    (:Authority
      (Utilities|IsValid _attachmentprop
        (:Is%20Valid
          (EventDispatchers|BindEventtoOnDied _attachmentprop (delegate Custom|Died)))))))

(event Custom|Died
  (params (DamagedActor :type Actor%20Object%20Reference) (Damage :type Float%20%28single-precision%29) (InstigatedBy :type Controller%20Object%20Reference) (DamageCauser :type Actor%20Object%20Reference) (HitLocation :type Vector) (FHitComponent :type Primitive%20Component%20Object%20Reference) (BoneName :type Name) (Momentum :type Vector))
  (bind _self self)
  (Utilities|IsValid _self
    (:Is%20Valid
      (Actor|DestroyActor _self))))

(event Interaction|EventBlueprintOnInteract
  (params (InteractingPawn :type Fort%20Pawn%20Object%20Reference) (InteractionBeingAttempted :type EInteractionBeingAttempted%20Enum :default "FirstInteraction"))
  (Utilities|Array|AddUnique (Variables|Default|GetListOfPawnsWhoHaveUsedHorn) InteractingPawn)
  (CallFunction|PlayBeep))

(event Custom|PlayBeep
  (Networking|SwitchHasAuthority
    (:Remote
      (Audio|Components|Audio|Play (Variables|Default|GetAudio) 0.0))))

(fn BlueprintCanInteract
  (params (InteractingPawn :type Fort%20Pawn%20Object%20Reference) (InteractionBeingAttempted :type EInteractionBeingAttempted%20Enum :default "FirstInteraction") (InteractionType :type TInteractionType%20Enum :default "IT_NoInteraction"))
  (if (Utilities|Array|ContainsItem (Variables|Default|GetListOfPawnsWhoHaveUsedHorn) InteractingPawn)
    (return false)
    (else
      (return true))))

(fn BlueprintGetInteractionString
  (params (InteractingPawn :type Fort%20Pawn%20Object%20Reference) (InteractionBeingAttempted :type EInteractionBeingAttempted%20Enum :default "FirstInteraction"))
  (return "NSLOCTEXT(\"[ED2C6FF808898B0B513EFA2B46E57325]\", \"D59AC16544D696857542A4AE821253DB\", \"Vote to Add Rocks\")"))

(fn ConstructionScript)
