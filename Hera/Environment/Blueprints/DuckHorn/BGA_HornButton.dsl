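; Interactable button that toots the Duck Boat horn, limited to one use per pawn and a maximum total count.
; Interaction records the pawn, counts the use, plays the attenuated horn sound and blocks presses for a six-second delay.
; Authority-bound death events disable the button when the horn prop dies and destroy it when its attachment actor dies.
; HornMaxUsesReached, set by exhausted uses or prop death, drives sparks and looping electric audio through its OnRep.
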(event EventBeginPlay
  (Networking|SwitchHasAuthority
    (:Authority
      (EventDispatchers|BindEventtoOnDied (Variables|Default|GetHornPropRef) (delegate Custom|Died) :!class /Script/FortniteGame.BuildingActor)
      (EventDispatchers|BindEventtoOnDied (Variables|Default|GetAttachmentActor) (delegate Custom|AttachmentActorDied) :!class /Script/FortniteGame.BuildingActor))))

(event Custom|Died
  (params (DamagedActor :type Actor%20Object%20Reference) (Damage :type Float%20%28single-precision%29) (InstigatedBy :type Controller%20Object%20Reference) (DamageCauser :type Actor%20Object%20Reference) (HitLocation :type Vector) (FHitComponent :type Primitive%20Component%20Object%20Reference) (BoneName :type Name) (Momentum :type Vector))
  (Variables|Default|SetHornPropAlive false)
  (Variables|Default|SetHornMaxUsesReached true))

(event Custom|AttachmentActorDied
  (params (DamagedActor :type Actor%20Object%20Reference) (Damage :type Float%20%28single-precision%29) (InstigatedBy :type Controller%20Object%20Reference) (DamageCauser :type Actor%20Object%20Reference) (HitLocation :type Vector) (FHitComponent :type Primitive%20Component%20Object%20Reference) (BoneName :type Name) (Momentum :type Vector))
  (Actor|DestroyActor self))

(event Interaction|EventBlueprintOnInteract
  (params (InteractingPawn :type Fort%20Pawn%20Object%20Reference) (InteractionBeingAttempted :type EInteractionBeingAttempted%20Enum :default "FirstInteraction"))
  (Variables|Default|SetAllowButtonPress false)
  (Utilities|FlowControl|DoOnce false
    (:Completed
      (Utilities|Array|Add (Variables|Default|GetListOfPawnsWhoHaveUsedHorn) InteractingPawn)
      (if (>= (Variables|Default|GetCurrentNumOfHornUses) (Variables|Default|GetMaxNumOfHornUses))
        (Variables|Default|SetHornMaxUsesReached true)
        (else
          (Math|Integer|IncrementInt (Variables|Default|GetCurrentNumOfHornUses)
            (:%20%20
              (CallFunction|PlayHornLocal)
              (Utilities|FlowControl|Delay 6.0)
              (CallFunction|ResetAllowHorn))))))))

(event Custom|ResetAllowHorn
  (Variables|Default|SetAllowButtonPress true)
  (Utilities|FlowControl|DoOnce false
    (:Completed
      (Utilities|Array|Add (Variables|Default|GetListOfPawnsWhoHaveUsedHorn) (stale Interaction|EventBlueprintOnInteract InteractingPawn))
      (if (>= (Variables|Default|GetCurrentNumOfHornUses) (Variables|Default|GetMaxNumOfHornUses))
        (Variables|Default|SetHornMaxUsesReached true)
        (else
          (Math|Integer|IncrementInt (Variables|Default|GetCurrentNumOfHornUses)
            (:%20%20
              (CallFunction|PlayHornLocal)
              (Utilities|FlowControl|Delay 6.0)
              (CallFunction|ResetAllowHorn))))))))

(event Custom|PlayHornLocal
  (bind _returnvalue (Audio|SpawnSoundatLocation :Sound "/Hera/Environment/Blueprints/DuckHorn/Assets/audio/Spot_Duck_Yatch_Horn_Cue.Spot_Duck_Yatch_Horn_Cue" :Location (Transformation|GetWorldLocation (Variables|Default|GetHornSound)) :VolumeMultiplier 1.0 :PitchMultiplier 1.0 :StartTime 0.0 :AttenuationSettings "/Game/Sounds/Attenuation/Fort_Tile_Ambient_LighthouseHornApollo_Att.Fort_Tile_Ambient_LighthouseHornApollo_Att" :bAutoDestroy true))
  (Audio|Components|Audio|Play _returnvalue 0.0))

(fn BlueprintCanInteract
  (params (InteractingPawn :type Fort%20Pawn%20Object%20Reference) (InteractionBeingAttempted :type EInteractionBeingAttempted%20Enum :default "FirstInteraction") (InteractionType :type TInteractionType%20Enum :default "IT_NoInteraction"))
  (if (Variables|Default|GetHornPropAlive)
    (if (Utilities|Array|ContainsItem (Variables|Default|GetListOfPawnsWhoHaveUsedHorn) InteractingPawn)
      _
      (elif (Variables|Default|GetHornMaxUsesReached)
        _
        (elif (Variables|Default|GetAllowButtonPress)
          (return true)
          (else
            (return false))))))
  (return false))

(fn BlueprintGetInteractionString
  (params (InteractingPawn :type Fort%20Pawn%20Object%20Reference) (InteractionBeingAttempted :type EInteractionBeingAttempted%20Enum :default "FirstInteraction"))
  (return "NSLOCTEXT(\"[FEA808FAB75192A2C1A17E11AA7E06E5]\", \"EB3D64EF49623F04837E7DB57AAD4924\", \"Toot Duck Boat Horn\")"))

(fn OnRep_HornMaxUsesReached
  (if (Variables|Default|GetHornMaxUsesReached)
    (Components|Activation|Activate (Variables|Default|GetNS_Env_XLCity_LightPost_Sparks) false)
    (Audio|Components|Audio|Play (Variables|Default|GetElemental_ElectricLoop_AOE_Cue) 0.0)
    (else
      (Components|Activation|Deactivate (Variables|Default|GetNS_Env_XLCity_LightPost_Sparks))
      (Audio|Components|Audio|Stop (Variables|Default|GetElemental_ElectricLoop_AOE_Cue)))))

(fn ConstructionScript)
